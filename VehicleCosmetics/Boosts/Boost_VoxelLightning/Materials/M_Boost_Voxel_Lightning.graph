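# Unlit, dither-masked material for Niagara mesh particles in the Voxel Lightning vehicle boost.
# UV channels 2–3 supply offset and variation data, which particle alpha and dynamic parameters use to displace, scale and rotate geometry around pivots.
# A distance-based "Sucky" blend mixes that motion with an object-position offset, behind an if_OW switch.
# Camera-facing shading tinted by particle color feeds emissive adjustments, alongside a camera depth fade.
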Material /VehicleCosmetics/Boosts/Boost_VoxelLightning/Materials/M_Boost_Voxel_Lightning
graph {
 properties {
  bAutomaticallySetUsageInEditor=false
  bHasPixelAnimation=true
  bUsedWithNiagaraMeshParticles=true
  blendMode=BLEND_Masked
  ditherOpacityMask=true
  shadingModel=MSM_Unlit
  usedShadingModels=Unlit
 }
 nodes [
  n1 = Constant()
  n2 = ObjectPositionWS()
  n3 = WorldPosition()
  n4 = AppendVector(A=TextureCoordinate() {coordinateIndex=2}, B=ComponentMask(None=TextureCoordinate() {coordinateIndex=3}) {r=true})
  n5 = ComponentMask(None=TextureCoordinate() {coordinateIndex=3}) {g=true}
  n6 = ParticleColor()
  n7 = Constant() {r=10}
  n8 = ComponentMask(None=TextureCoordinate() {coordinateIndex=3}) {g=true}
  n9 = LinearInterpolate(A=Constant3Vector() {constant=<rgba 0 1 0 0>}, B=Constant3Vector() {constant=<rgba 1 -1 0 0>}, Alpha=Floor(None=Multiply(A=@n7, B=@n8)))
  n10 = WorldPosition()
  n11 = ComponentMask(None=TextureCoordinate() {coordinateIndex=3}) {g=true}
  n12 = DynamicParameter() {paramNames [Scale,ScaleRand,RotAmount,Contrast]}
  n13 = Multiply(
   A=Subtract(
    A=@n10.XYZ
    B=MaterialFunctionCall() {materialFunction=/VehicleCosmetics/Boosts/Boost_SkullChromatic/Materials/MF_ObjectPivotPoint}.ObjectPivotLocation
   )
   B=LinearInterpolate(A=Constant() {r=-1}, B=LinearInterpolate(B=@n11, Alpha=@n12.ScaleRand), Alpha=@n12.Scale)
  )
  n14 = Multiply(A=Frac(None=DynamicParameter() {paramNames [Scale,ScaleRand,RotAmount,Contrast]}.RotAmount), B=Constant() {r=6.28})
  n15 = Add(
   A=Multiply(A=Transform(None=@n4) {transformSourceType=TRANSFORMSOURCE_Local}, B=Add(A=@n5, B=@n6.A))
   B=RotateAboutAxis(
    NormalizedRotationAxis=Normalize(VectorInput=@n9)
    RotationAngle=Saturate(None=@n14)
    PivotPoint=MaterialFunctionCall() {materialFunction=&MF_ObjectPivotPoint}.ObjectPivotLocation
    Position=Add(A=@n10.XYZ, B=@n13)
   )
  )
  n16 = WorldPosition()
  n17 = Power(
   Base=Add(A=Distance(A=@n2, B=@n16.XYZ), B=Constant() {r=2048})
   Exp=ScalarParameter() {defaultValue=12, group=Sucky, parameterName=SuckyGradFalloff}
  )
  If(
   A=ScalarParameter() {defaultValue=1, parameterName=if_OW}
   B=@n1
   "A > B"=LinearInterpolate(A=Subtract(A=@n2, B=@n3.XYZ), B=Add(A=@n15, B=@n13), Alpha=Saturate(None=@n17))
   "A == B"=@n1
   "A < B"=@n1
  )
  n18 = CameraVectorWS()
  MaterialExpressionConstantClamp_13 = Saturate(None=DotProduct(
    A=Transform(None=@n18) {transformSourceType=TRANSFORMSOURCE_World, transformType=TRANSFORM_Tangent}
    B=Constant3Vector() {constant=<rgba 0 0 1 0>}
   ))
  MaterialExpressionMeshEmitterVertexColor_0 = ParticleColor()
  MaterialFunctionCall(Emissive=Multiply(A=@MaterialExpressionConstantClamp_13, B=@MaterialExpressionMeshEmitterVertexColor_0.RGB)) {materialFunction=&MF_FX_EmissiveAdjustments}
  MaterialFunctionCall(Fade Length=Constant() {r=160}, Fade Offset=Constant() {r=80}) {materialFunction=&CameraDepthFade}
 ]
}
